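# Niagara swing trail effect for the ChessBoard pickaxe, drawn as short-lived ribbons that follow the weapon's animation.
# Two AnimTrail emitters, one blue-grey and one fainter, shorter-lived green, sample two skeletal mesh bones or sockets and each feed two ribbon renderers.
# Time-of-day selectors set and tint each color, then curves scale it and fade alpha out.
# Velocity-add and ribbon-width taper modules are present but disabled.
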NiagaraSystem /BRCosmetics/Weapons/FORT_Melee/Pickaxe_ChessBoard/FX/NS_Pickaxe_ChessBoard_Swing
dependencies {
 usedDataInterfaces [&NiagaraDataInterfaceColorCurve, &NiagaraDataInterfaceCurve, &NiagaraDataInterfaceSkeletalMesh]
 usedDynamicInputs []
 usedModules [
  &self:SystemSpawnScript.SystemScriptSource.SystemScriptGraph.NiagaraNodeAssignment_0.SetVariables_#1
  &SystemState
 ]
}
emitters {
 FN_AnimTrail002 {
  emitterSpawnScript [&FN_RibbonIDOffsetUtility()]
  emitterUpdateScript [
   &EmitterState()
   &FN_SpawnAnimTrail()
  ]
  particleSpawnScript [
   &FN_InitializeAnimTrailBoneIndecies(First Bone or Socket Index=1, Second Bone or Socket Index=0)
   &FN_InitializeAnimTrail(
    Color=&FNTimeOfDayColor(
     Night Color=<rgba 0.2464 0.2821 0.35 1>
     Morning Color=<rgba 0.2464 0.2821 0.35 0.3>
     Day Color=<rgba 0.2816 0.3224 0.4 1>
     Evening Color=<rgba 0.2464 0.2821 0.35 0.3>
    )
    Lifetime=0.18
    SkeletalMesh=$System.Skeletal Mesh
   )
   &Color(Color=&MultiplyLinearColor(
     Linear Color=$Particles.Initial.Color
     Scale Factor=&MakeVector4_FromLinearColor(LinearColor=&FN_Select_Color_from_Time_Of_Day(Morning=<rgba 0.6719 0.6019 0.6097 1>, Day=<rgba 1 1 1 1.1>, Evening=<rgba 0.753 0.4829 0.5991 1>, Night=<rgba 0.4085 0.7931 1.75 1>))
    )) {module=Color001}
   &AddVelocity(Velocity=<0 400 0>) {enabled=false}
  ]
  particleUpdateScript [
   &SolveForcesAndVelocity(Clamp Velocity=false, Limit Acceleration=false)
   &FN_UpdateAnimTrail(Locked Particle RibbonWidth=1, SkeletalMesh=$System.Skeletal Mesh) {module=FN_UpdateAnimTrail001}
   &ScaleColor(
    Linear Color Curve=DataInterface() {propertyValues {
      AlphaCurve {keys [<at 0 0.7>,<at 0.038 1>,<at 0.1439 1>,<at 0.3671 0.6>,<at 0.5309 0.25>,<at 0.7343 0>]}
      BlueCurve {keys [<at 0 0.349>,<at 0.5969 0.5208>]}
      GreenCurve {keys [<at 0 0.349>,<at 0.5969 0.5155>]}
      RedCurve {keys [<at 0 0.349>,<at 0.5969 0.5087>]}
     }}
    Scale Mode=::NewEnumerator2
   )
   &ScaleRibbonWidth(Ribbon Width Scale=&FloatFromCurve(FloatCurve=DataInterface() {propertyValues {Curve {keys [
         {arriveTangent=0.3072, interpMode=RCIM_Cubic, leaveTangent=0.3072, tangentMode=RCTM_User, time=0, value=1}
         {arriveTangent=-0.7535, interpMode=RCIM_Cubic, leaveTangent=-0.7535, tangentMode=RCTM_User, time=1, value=0}
        ]}}})) {enabled=false}
  ]
  rendererClasses [&NiagaraRibbonRendererProperties]
  renderers [
   {rendererClass=&NiagaraRibbonRendererProperties}
   {rendererClass=&NiagaraRibbonRendererProperties, rendererIndex=1}
  ]
 }
 FN_AnimTrail003 {
  emitterSpawnScript [&FN_RibbonIDOffsetUtility()]
  emitterUpdateScript [
   &EmitterState()
   &FN_SpawnAnimTrail()
  ]
  particleSpawnScript [
   &FN_InitializeAnimTrailBoneIndecies(First Bone or Socket Index=1, Second Bone or Socket Index=0)
   &FN_InitializeAnimTrail(
    Color=&FNTimeOfDayColor(
     Night Color=<rgba 0.3332 0.5 0.2875 0.05>
     Morning Color=<rgba 0.0666 0.1 0.0575 0.05>
     Day Color=<rgba 0.2998 0.45 0.2587 0.05>
     Evening Color=<rgba 0.0666 0.1 0.0575 0.05>
    )
    Lifetime=0.16
    SkeletalMesh=$System.Skeletal Mesh
   )
   &Color(Color=&MultiplyLinearColor(
     Linear Color=$Particles.Initial.Color
     Scale Factor=&MakeVector4_FromLinearColor(LinearColor=&FN_Select_Color_from_Time_Of_Day(Morning=<rgba 0.6719 0.6019 0.6097 1>, Day=<rgba 1 1 1 0.9>, Evening=<rgba 1 0.6413 0.7956 1>, Night=<rgba 0.3501 0.6798 1.5 1>))
    )) {module=Color001}
   &AddVelocity(Velocity=<0 400 0>) {enabled=false}
  ]
  particleUpdateScript [
   &SolveForcesAndVelocity(Clamp Velocity=false, Limit Acceleration=false)
   &FN_UpdateAnimTrail(Locked Particle RibbonWidth=1, SkeletalMesh=$System.Skeletal Mesh) {module=FN_UpdateAnimTrail001}
   &ScaleColor(
    Linear Color Curve=DataInterface() {propertyValues {
      AlphaCurve {keys [<at 0 0.7>,<at 0.038 1>,<at 0.129 1>,<at 0.2754 0.3>,<at 0.4961 0.05>,<at 0.7442 0>]}
      BlueCurve {keys [<at 0 0.4948>,<at 0.5969 0.5208>]}
      GreenCurve {keys [<at 0 0.4948>,<at 0.5969 0.5208>]}
      RedCurve {keys [<at 0 0.4948>,<at 0.5969 0.5208>]}
     }}
    Scale Mode=::NewEnumerator2
   )
   &ScaleRibbonWidth(Ribbon Width Scale=&FloatFromCurve(FloatCurve=DataInterface() {propertyValues {Curve {keys [
         {arriveTangent=0.3072, interpMode=RCIM_Cubic, leaveTangent=0.3072, tangentMode=RCTM_User, time=0, value=1}
         {arriveTangent=-1.4466, interpMode=RCIM_Cubic, leaveTangent=-1.4466, tangentMode=RCTM_User, time=1, value=0}
        ]}}})) {enabled=false}
  ]
  rendererClasses [&NiagaraRibbonRendererProperties]
  renderers [
   {rendererClass=&NiagaraRibbonRendererProperties}
   {rendererClass=&NiagaraRibbonRendererProperties, rendererIndex=1}
  ]
 }
}
summary {
 systemName=NS_Pickaxe_ChessBoard_Swing
 emitters [
  {emitterName=FN_AnimTrail002, rendererClasses [&NiagaraRibbonRendererProperties]}
  {emitterName=FN_AnimTrail003, rendererClasses [&NiagaraRibbonRendererProperties]}
 ]
 userVariables [User.RibbonIDOffset {type=NiagaraInt32}]
}
